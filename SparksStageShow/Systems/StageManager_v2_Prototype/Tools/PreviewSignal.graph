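# PreviewSignal is a prototype editor utility widget in the Stage Manager tools, laid out to preview a signal through an image using the M_SignalPreview material.
# A stack box holds that fixed-size image, a radial slider, a property view and an editor slider.
# PreConstruct binds the property view to the widget's own TestBool variable.
# Construct and Tick are empty, and no logic connects the sliders to the preview.
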EditorUtilityWidgetBlueprint /SparksStageShow/Systems/StageManager_v2_Prototype/Tools/PreviewSignal
graphs [EventGraph {nodes [
   Event PreConstruct()
   n1 = Get SinglePropertyView_89() -> (SinglePropertyView_89:Single Property View Object Reference)
   > SetObject(self=@n1, NewObject=Self-Reference())
   > SetPropertyName!(execute, self:Single Property View Object Reference=@n1, NewPropertyName:Name=TestBool) -> (then)
   SetValueObject!(execute, Property:Property Value Object Reference, InValue:Object Reference) -> (then)
   Event Construct()
   Event Tick()
   Get TestBool() -> (TestBool:bool)
  ]}]
variables [TestBool {type=boolean}]
widgets {
 info {rootWidgetClass=&StackBox, widgetCount=7}
 widgets [
  StackBox_0 {widgetClassPath=&StackBox, widget=StackBox_0}
  RadialSlider_74 {widgetClassPath=/Script/AdvancedWidgets.RadialSlider, bIsVariable=true, parent=StackBox_0, widget=RadialSlider_74, slot {class=StackBoxSlot}}
  SizeBox_37 {
   widgetClassPath=&SizeBox
   parent=StackBox_0
   widget=SizeBox_37
   properties {bOverride_HeightOverride=true, bOverride_WidthOverride=true, heightOverride=200, widthOverride=500}
   slot {class=StackBoxSlot, properties {horizontalAlignment=HAlign_Left, verticalAlignment=VAlign_Top}}
  }
  Image_0 {
   widgetClassPath=&Image
   bIsVariable=true
   parent=SizeBox_37
   widget=Image_0
   properties {brush {
     bIsDynamicallyLoaded=false
     drawAs=Image
     imageType=FullColor
     mirroring=NoMirror
     resourceName=None
     resourceObject=&/M_SignalPreview
     tiling=NoTile
     imageSize=<32 32>
     margin=<ltrb 0 0 0 0>
     outlineSettings {
      bUseBrushTransparency=false
      roundingType=HalfHeightRadius
      width=0
      color {colorUseRule=UseColor_Specified, specifiedColor=<rgba 0 0 0 0>}
      cornerRadii=<0 0 0 0>
     }
     tintColor {colorUseRule=UseColor_Specified, specifiedColor=<rgba 1 1 1 1>}
     uVRegion {bIsValid=false, max=<0 0>, min=<0 0>}
    }}
   slot {class=SizeBoxSlot}
  }
  SinglePropertyView_89 {
   widgetClassPath=&SinglePropertyView
   parent=StackBox_0
   widget=SinglePropertyView_89
   properties {propertyName=TestBool}
   slot {class=StackBoxSlot}
  }
  SizeBox_42 {
   widgetClassPath=&SizeBox
   parent=StackBox_0
   widget=SizeBox_42
   properties {bOverride_WidthOverride=true, widthOverride=200}
   slot {class=StackBoxSlot}
  }
  EditorUtilitySlider_34 {
   widgetClassPath=/Script/Blutility.EditorUtilitySlider
   bIsVariable=true
   parent=SizeBox_42
   widget=EditorUtilitySlider_34
   slot {class=SizeBoxSlot}
  }
 ]
}
